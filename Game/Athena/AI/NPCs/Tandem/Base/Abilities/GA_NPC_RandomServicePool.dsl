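; Gameplay ability for Tandem NPCs that gives the avatar pawn one random tag from a service pool.
; Once playlist data is ready, it commits and keeps pool tags whose Tandem_Service_Data curve row evaluates above zero.
; If the pawn has none of the full pool's tags, it adds a random filtered one as a replicated loose tag.
; It ends itself only when the pawn cast fails.
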(event Ability|EventActivateAbility
  (Class|FortAsyncActionCurrentPlaylistReady|PlaylistDataReadyAsync
    (:OnReady
      (Ability|CommitAbility)
      (bind _asfort_player_pawn (Utilities|Casting|CastToFortPlayerPawn (Ability|GetAvatarActorfromActorInfo))
        (:then
          (Variables|Default|SetFort%20Player%20Pawn _asfort_player_pawn)
          (for _array_element (Variables|Default|GetGameplay%20Tags)
            (bind _outxy (DataTable|EvaluateCurveTableRow "/Game/Athena/AI/NPCs/Tandem/Base/DataTables/Tandem_Service_Data.Tandem_Service_Data" (GameplayTags|GetTagName _array_element) 1.0)
              (:RowFound
                (if (> _outxy 0)
                  (Utilities|Array|Add (Variables|Default|GetGameplay%20Tags_Filtered) _array_element)))))
          (if (not (GameplayTags|HasAnyMatchingGameplayTags (Variables|Default|GetFort%20Player%20Pawn) (GameplayTags|MakeGameplayTagContainerfromArray (Variables|Default|GetGameplay%20Tags))))
            (Ability|GameplayEffect|AddLooseGameplayTags (Variables|Default|GetFort%20Player%20Pawn) (GameplayTags|MakeGameplayTagContainerfromTag (Utilities|Array|RandomArrayItem (Variables|Default|GetGameplay%20Tags_Filtered))) true)))
        (:CastFailed
          (Ability|EndAbility))))))

(event Ability|EventOnEndAbility
  (params (bWasCancelled :type Boolean))
  (|CallParentFunction bWasCancelled) ; This node is disabled and will not be called.\nDrag off pins to build functionality.
)
